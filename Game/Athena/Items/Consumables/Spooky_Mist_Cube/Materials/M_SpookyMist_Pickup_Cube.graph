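# Lit material for the Spooky Mist Cube consumable pickup, with animated glowing runes, routed through MF_CreativeMoveTool.
# Emissive masks are tinted by layered-sine flicker and a panning vertical gradient, plus a pulsing SphereMask glow around the pivot.
# UseWPO bobs and spins the cube about local Z in World Position Offset, rotating normals to match.
# Static switches offer pre-skinned positions, particle-color bypass for skeletal use, and Fresnel-masked emissive.
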Material /Game/Athena/Items/Consumables/Spooky_Mist_Cube/Materials/M_SpookyMist_Pickup_Cube
graph {
 properties {
  bUseMaterialAttributes=true
  bUsedWithMeshDeformer=true
  bUsedWithSkeletalMesh=true
  pixelDepthOffsetMode=PDOM_Legacy
  usedShadingModels=Default Lit
 }
 nodes [
  n1 = TextureSampleParameter2D() {group=Emissive, parameterName=EmissiveMasks, texture=/Game/Athena/Items/Consumables/Spooky_Mist_Cube/Textures/T_Spooky_Mist_Cube_E_V2}
  n2 = LinearInterpolate(Alpha=@n1.R) {constA=1, constB=0.5}
  n3 = TextureSampleParameter2D() {
   desc="R = SpecMask G = Metallic B = Roughness"
   group=Specular
   parameterName=SpecularMasks
   samplerSource=SSM_Wrap_WorldGroupSettings
   samplerType=SAMPLERTYPE_Masks
   texture=/Game/Athena/Items/Consumables/Spooky_Mist_Cube/Textures/Spooky_Mist_Cube_S
  }
  n4 = ScalarParameter() {group=Roughness, parameterName=SpecRoughnessMin}
  n5 = ScalarParameter() {defaultValue=1, group=Roughness, parameterName=SpecRoughnessMax}
  n6 = Add(A=Multiply(A=Time(), B=ScalarParameter() {defaultValue=1, parameterName=Cube Rune Flicker Speed}), B=@n1.G)
  n7 = Add(A=Multiply(A=Sine(None=@n6) {period=0.22}, B=Sine(None=@n6) {period=0.11}), B=Sine(None=@n6) {period=0.65})
  n8 = ConstantBiasScale(None=Sine(None=@n6) {period=3})
  n9 = Multiply(A=Multiply(A=Clamp(None=@n7) {maxDefault=2}) {constB=2}, B=@n8)
  n10 = Power(Base=@n8) {constExponent=3}
  n11 = VertexInterpolator(VS=PreSkinnedPosition())
  n12 = StaticSwitchParameter(True=@n11.PS, False=WorldPosition().XYZ) {desc=(instead of Absolute World Position), parameterName=Use Pre-Skinned Position}
  n13 = Time()
  n14 = ScalarParameter() {defaultValue=1, parameterName=Random}
  n15 = Add(A=@n13, B=@n14)
  n16 = Add(
   A=Divide(A=ComponentMask(None=@n12) {b=true}, B=ScalarParameter() {defaultValue=50, parameterName=GradientTile}) {constB=200}
   B=Multiply(A=ScalarParameter() {defaultValue=-0.27, parameterName=GradientPanSpeed}, B=@n15) {constB=-0.2}
  )
  n17 = LinearInterpolate(Alpha=MaterialFunctionCall(Speed=ScalarParameter() {defaultValue=0.35, parameterName=CraterPulseSpeed}) {materialFunction=&DebugTimeSine}.Result) {constA=0.05, constB=0.2}
  n18 = StaticSwitchParameter(
   True=Multiply(
    A=LinearInterpolate(A=@n9, Alpha=@n10)
    B=TextureSample(UVs=@n16) {texture=/Game/Global/Textures/Default/Masks_and_Gradients/T_VertGradient_02_sRGB}.G
   )
   False=@n17
  ) {defaultValue=true, parameterName=CubeEmissive}
  n19 = VectorParameter() {group=Emissive, parameterName=EmissiveEdgeColor, defaultValue=<rgba 1.0742 0.1719 3 0>}
  n20 = Multiply(
   A=@n1.R
   B=LinearInterpolate(A=@n19.RGB, B=VectorParameter() {group=Emissive, parameterName=EmissiveColor, defaultValue=<rgba 4.6976 0.1302 5 0>}.RGB, Alpha=@n1.R)
  )
  n21 = ScalarParameter() {defaultValue=24, parameterName=GlowRadius}
  n22 = ScalarParameter() {parameterName=GlowHardness}
  n23 = SphereMask(
   A=@n11.PS
   B=ComponentMask(None=VectorParameter() {parameterName=Pre-Skinned Pivot Offset}.RGB) {b=true, g=true, r=true}
   Radius=@n21
   Hardness=@n22
  ) {hardnessPercent=30}
  n24 = ActorPositionWS()
  n25 = ObjectPositionWS()
  n26 = Multiply(A=Sine(None=@n15) {period=3.3}, B=Transform(None=Constant3Vector() {constant=<rgba 0 0 3 0>}) {transformSourceType=TRANSFORMSOURCE_Local})
  n27 = StaticSwitchParameter(True=@n26, False=Constant()) {defaultValue=true, parameterName=UseWPO}
  n28 = Transform(None=VectorParameter() {parameterName=PivotOffset, defaultValue=<rgba 0 0 44 0>}.RGB) {transformSourceType=TRANSFORMSOURCE_Local}
  n29 = SphereMask(
   A=WorldPosition().XYZ
   B=Add(A=StaticSwitchParameter(True=@n24, False=@n25) {defaultValue=true, parameterName=UseWPO}, B=Add(A=@n27, B=@n28))
   Radius=@n21
   Hardness=@n22
  ) {hardnessPercent=30}
  n30 = Multiply(
   A=LinearInterpolate(Alpha=@n3.B) {constA=1.5, constB=0.4}
   B=Power(
    Base=StaticSwitchParameter(True=@n23, False=@n29) {desc=(instead of Absolute World Position), parameterName=Use Pre-Skinned Position}
    Exp=ScalarParameter() {defaultValue=2, parameterName=GlowPower}
   )
  )
  n31 = ScalarParameter() {defaultValue=2, parameterName=Brightness}
  n32 = Multiply(A=@n19.RGB, B=Multiply(A=@n17, B=Multiply(A=@n30, B=@n31)) {constB=0.2})
  n33 = Constant()
  n34 = Multiply(
   A=StaticSwitchParameter(True=Constant() {r=1}, False=ParticleColor().RGB) {parameterName=ApplytoSkeletal}
   B=Add(A=Multiply(A=@n18, B=@n20), B=StaticSwitchParameter(True=@n32, False=@n33) {defaultValue=true, parameterName=CubeEmissive})
  )
  n35 = LinearInterpolate(B=@n34, Alpha=@n1.B)
  n36 = TextureSampleParameter2D() {parameterName=Normals, samplerType=SAMPLERTYPE_Normal, texture=/Game/Athena/Items/Consumables/Spooky_Mist_Cube/Textures/Spooky_Mist_Cube_N}
  n37 = Fresnel(ExponentIn=ScalarParameter() {defaultValue=4, group=Emissive, parameterName=Emissive_FresEX}, Normal=Transform(None=@n36.RGB)) {baseReflectFraction=0}
  n38 = StaticSwitchParameter(True=Multiply(A=@n35, B=OneMinus(None=@n37)), False=@n35) {group=Emissive, parameterName=Use Fresnel in Mask}
  n39 = Normalize(VectorInput=Transform(None=Constant3Vector() {constant=<rgba 0 0 1 0>}) {transformSourceType=TRANSFORMSOURCE_Local})
  n40 = Multiply(A=@n13, B=Multiply(A=@n14, B=ScalarParameter() {defaultValue=0.1, parameterName=RotateSpeed}) {constB=0.2}) {constB=0.1}
  n41 = MaterialFunctionCall(
   BaseNormal=MaterialFunctionCall(Rotation Axis=@n39, Rotation Angle=@n40) {materialFunction=&FixRotateAboutAxisNormals}.New Tangent Space Vertex Normal
   AdditionalNormal=@n36.RGB
  ) {materialFunction=&BlendAngleCorrectedNormals}
  n42 = Add(A=ActorPositionWS(), B=@n28)
  n43 = Add(A=@n26, B=RotateAboutAxis(NormalizedRotationAxis=@n39, RotationAngle=@n40, PivotPoint=Add(A=@n26, B=@n42), Position=WorldPosition().XYZ))
  n44 = Constant()
  MaterialFunctionCall(MA In=MakeMaterialAttributes!(
    BaseColor=TextureSampleParameter2D() {parameterName=Diffuse, texture=/Game/Athena/Items/Consumables/Spooky_Mist_Cube/Textures/Spooky_Mist_Cube_D}.RGB
    Metallic=@n3.G
    Specular=Multiply(A=@n2, B=@n3.R)
    Roughness=LinearInterpolate(A=@n4, B=@n5, Alpha=@n3.B)
    Anisotropy
    EmissiveColor=StaticSwitchParameter(True=@n38, False=@n34) {group=Emissive, parameterName=Mask Composite Emissive}
    Opacity
    OpacityMask
    Normal=StaticSwitchParameter(True=@n41.Result, False=@n36.RGB) {defaultValue=true, parameterName=UseWPO}
    Tangent
    WorldPositionOffset=StaticSwitchParameter(True=@n43, False=@n44) {defaultValue=true, parameterName=UseWPO}
    SubsurfaceColor
    ClearCoat
    ClearCoatRoughness
    AmbientOcclusion
    Refraction
    CustomizedUVs_0
    CustomizedUVs_1
    CustomizedUVs_2
    CustomizedUVs_3
    CustomizedUVs_4
    CustomizedUVs_5
    CustomizedUVs_6
    CustomizedUVs_7
    PixelDepthOffset
    ShadingModel
    Displacement
   )) {materialFunction=&MF_CreativeMoveTool}
 ]
}
